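# Material function that builds base Morphite weapon material attributes: a metallic, emissive, iridescent surface with overridable color, normals and UVs.
# Base color combines a Fresnel hue-shifted tint with triangle-texture-driven curve-atlas reflections, and the weapon diffuse is reworked through desaturation and color dodge.
# Hard-edge world-aligned UVs, vertex-normal direction masks and sky-light-direction masks support shading.
# Shading-path and ray-tracing switches simplify mobile color and vary roughness.
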MaterialFunction /BlastBerryWeapons/Materials/Morphite/MF_Morphite_Base
graph {nodes [
  n1 = SetMaterialAttributes!(
   MaterialAttributes
   BaseColor=NamedRerouteUsage()
   Metallic=NamedRerouteUsage()
   Specular=ScalarParameter() {group=Morphite, parameterName=Specular}
   Roughness=NamedRerouteUsage()
   OpacityMask=Constant() {r=1}
   Normal=NamedRerouteUsage()
   AmbientOcclusion=NamedRerouteUsage()
   EmissiveColor=NamedRerouteUsage()
  ) {attributeSetTypes [
    #1
    #2
    #3
    #4
    #5
    #6
    #7
    #8
   ]}
  FunctionOutput(None=@n1) {outputName=MA Out}
  n2 = VectorParameter() {group=Morphite, parameterName=BaseColor, defaultValue=<rgba 0 0.2582 0.5149 0>}
  NamedRerouteDeclaration(None=Multiply(A=VertexInterpolator(VS=@n2.RGB).PS, B=ScalarParameter() {defaultValue=4, group=Morphite, parameterName=EmissiveIntensity})) {name=Emissive, nodeColor=<rgba 0.0781 0 1 1>}
  n3 = Normalize(VectorInput=LinearInterpolate(
    A=FunctionInput() {bUsePreviewValueAsDefault=true, inputName=Override Normals}
    B=Constant3Vector() {constant=<rgba 0 0 1 0>}
    Alpha=ScalarParameter() {defaultValue=0.5, group=Morphite, parameterName=BaseNormalStrength}
   ))
  NamedRerouteDeclaration(None=@n3) {name=Normals, nodeColor=<rgba 1 0 0.9141 1>}
  n4 = TextureSampleParameter2D!(UVs, Level, Apply View MipBias) {
   group=Weapon_Textures
   mipValueMode=TMVM_MipLevel
   parameterName=Diffuse
   samplerSource=SSM_Wrap_WorldGroupSettings
   texture=/Game/Weapons/FORT_Shotguns/Textures/T_HammerShotGun_D
  }
  n5 = Desaturation(None=@n4.RGB, Fraction=ScalarParameter() {defaultValue=0.92, group=Morphite, parameterName=DiffuseSourceDesaturation})
  n6 = Add(
   A=Divide(
    A=VectorParameter() {group=Morphite, parameterName=DiffuseSource_ColorDodgeBase, defaultValue=<rgba 0.1476 0.1364 0.1284 0>}.RGB
    B=OneMinus(None=@n5)
   )
   B=ScalarParameter() {defaultValue=-0.5, group=Morphite, parameterName=DiffuseSourceBias}
  )
  n7 = ScalarParameter() {defaultValue=1.6, group=Morphite, parameterName=DiffuseSourceScale}
  NamedRerouteDeclaration(None=Saturate(None=Multiply(A=@n6, B=@n7))) {name=Diffuse, nodeColor=<rgba 1 0 0.1172 1>}
  n8 = FunctionInput(Preview=WorldPosition() {worldPositionShaderOffset=WPT_ExcludeAllShaderOffsets}.XYZ) {bUsePreviewValueAsDefault=true, inputName=WorldPositionOverride, sortPriority=3}
  n9 = NamedRerouteUsage()
  NamedRerouteDeclaration(None=FunctionInput(Preview=MaterialFunctionCall!(UseLocalSpace, WorldPosition=@n8, NormalOverride=@n9) {materialFunction=/Game/Environments/Apollo/Sets/Skyfire/Materials/MaterialFunctions/MF_WorldAlignedUVs_HardEdge}.Result) {bUsePreviewValueAsDefault=true, inputName=UVsOverride, inputType=FunctionInput_Vector2}) {name=UVs, nodeColor=<rgba 0.2891 0 1 1>}
  n10 = DotProduct(A=ReflectionVectorWS(CustomWorldNormal=PixelNormalWS()), B=SkyAtmosphereLightDirection())
  NamedRerouteDeclaration(None=Power(Base=Saturate(None=@n10), Exp=ScalarParameter() {defaultValue=40, group=Morphite, parameterName=ReflectionVectorExponent})) {name=SpecLightMask, nodeColor=<rgba 0 1 0.5312 1>}
  n11 = ScalarParameter() {defaultValue=2.5, group=Morphite, parameterName=EdgeNormal_FresnelExponent}
  n12 = NamedRerouteUsage()
  NamedRerouteDeclaration(None=LinearInterpolate(Alpha=MaterialFunctionCall(Exponent=@n11, Vertex / Pixel Normal=@n12) {materialFunction=&MF_Fresnel_FX}.Fresnel Inverted) {constA=0.05}) {name=EdgeNormalFresnelInverted, nodeColor=<rgba 0 1 0.5547 1>}
  n13 = ScalarParameter() {defaultValue=0.45, group=Morphite, parameterName=SunMoonMaskLow}
  n14 = ScalarParameter() {defaultValue=1, group=Morphite, parameterName=SunMoonMaskHigh}
  n15 = MaterialFunctionCall(
   Brightness=ScalarParameter() {defaultValue=1.7, group=Morphite, parameterName=SunMoonFresnelBrightness}
   Exponent=ScalarParameter() {defaultValue=2, group=Morphite, parameterName=SunMoonFresnelExponent}
   Camera Vector=SkyAtmosphereLightDirection()
   Vertex / Pixel Normal=NamedRerouteUsage()
  ) {materialFunction=&MF_Fresnel_FX}
  NamedRerouteDeclaration(None=ShadingPathSwitch(Default=LinearInterpolate(A=@n13, B=@n14, Alpha=@n15.Fresnel), Mobile=Constant() {r=1})) {name=SunMoonMaskRemapped, nodeColor=<rgba 1 0 0.5156 1>}
  NamedRerouteDeclaration(None=Divide(A=NamedRerouteUsage(), B=ScalarParameter() {defaultValue=1290.24, group=Morphite, parameterName=Chrome_Divisor_UV_a})) {name=Chrome_UVs_A_NoPanning, nodeColor=<rgba 0 1 0.7656 1>}
  n16 = DotProduct(A=NamedRerouteUsage(), B=Constant3Vector() {constant=<rgba 0 0 -1 0>})
  NamedRerouteDeclaration(None=SphereMask(A=Saturate(None=@n16), B=Constant() {r=0.1}, Radius=Constant() {r=0.2}, Hardness=Constant() {r=0.5})) {name=VertexNormalMask.-Z, nodeColor=<rgba 0 0.3203 1 1>}
  n17 = ComponentMask(None=NamedRerouteUsage()) {g=true, r=true}
  n18 = ScalarParameter() {defaultValue=0.5, group=Morphite, parameterName=ReflectionCurveUVDistortionScale}
  n19 = TextureSample(UVs=Add(A=Multiply(A=@n17, B=@n18), B=NamedRerouteUsage())) {
   samplerSource=SSM_Wrap_WorldGroupSettings
   samplerType=SAMPLERTYPE_LinearColor
   texture=/Game/TechArt/Seasonal/LackLuster/Textures/T_Lackluster_Triangles
  }
  n20 = CurveAtlasRowParameter(CurveTime=@n19.B) {
   atlas=/Game/TechArt/Seasonal/LackLuster/Curves/CA_Lackluster
   curve=/Game/TechArt/Seasonal/LackLuster/Curves/C_LackLuster_Spectrum_05
   defaultValue=13
   group=Morphite
   parameterName=ReflectionCurve
  }
  n21 = NamedRerouteUsage()
  n22 = Multiply(
   A=NamedRerouteDeclaration(None=Multiply(A=@n20, B=@n21)) {name=ReflectionCurve, nodeColor=<rgba 1 0.9609 0 1>}
   B=ScalarParameter() {defaultValue=5, group=Morphite, parameterName=ReflectionCurveBaseColorScale}
  )
  n23 = Transform(None=NamedRerouteUsage())
  n24 = Multiply(
   A=MaterialFunctionCall(
    Brightness=ScalarParameter() {defaultValue=0.75, group=Morphite, parameterName=FresnelHueShiftBrightness}
    Exponent=ScalarParameter() {defaultValue=1.5, group=Morphite, parameterName=FresnelHueShiftExponent}
    Vertex / Pixel Normal=NamedRerouteDeclaration(None=@n23) {name=NormalsWorldSpace, nodeColor=<rgba 1 0.5625 0 1>}
   ) {materialFunction=&MF_Fresnel_FX}.Fresnel Unclamped
   B=ScalarParameter() {defaultValue=0.2, group=Morphite, parameterName=FresnelHueShiftScale}
  )
  n25 = ScalarParameter() {defaultValue=-0.08, group=Morphite, parameterName=FresnelHueShiftOffset}
  n26 = VectorParameter() {group=Morphite, parameterName=BaseColor, defaultValue=<rgba 0 0.3578 0.7135 0>}
  n27 = FunctionInput(Preview=MaterialFunctionCall(Hue Shift Percentage=Add(A=@n24, B=@n25), Texture=VertexInterpolator(VS=@n26.RGB).PS) {materialFunction=&HueShift}.Result) {bUsePreviewValueAsDefault=true, inputName=BaseColorOverride, sortPriority=2}
  n28 = Add(A=ShadingPathSwitch(Default=Add(A=@n22, B=@n27), Mobile=@n27), B=Multiply(A=NamedRerouteUsage(), B=NamedRerouteUsage()))
  n29 = ScalarParameter() {defaultValue=0.1, group=Morphite, parameterName=BaseColorDesaturationFraction}
  n30 = Multiply(
   A=@n28
   B=LinearInterpolate(A=ScalarParameter() {defaultValue=0.25, group=Morphite, parameterName=MobileShadingLow}, Alpha=NamedRerouteUsage())
  )
  NamedRerouteDeclaration(None=ShadingPathSwitch(Default=Desaturation(None=@n28, Fraction=@n29), Forward=@n30, Mobile=@n30)) {name=BaseColor, nodeColor=<rgba 0 0.6484 1 1>}
  n31 = RayTracingQualitySwitch(
   Normal=LinearInterpolate(
    A=ScalarParameter() {defaultValue=0.28, group=Morphite, parameterName=Roughness.-Z}
    B=ScalarParameter() {defaultValue=0.22, group=Morphite, parameterName=Roughness}
    Alpha=NamedRerouteUsage()
   )
   RayTraced=ScalarParameter() {defaultValue=0.5, group=Morphite, parameterName=Roughness_Raytraced}
  )
  n32 = LinearInterpolate(
   A=ScalarParameter() {defaultValue=0.15, group=Morphite, parameterName=Roughness_Mobile.-Z}
   B=ScalarParameter() {defaultValue=0.3, group=Morphite, parameterName=Roughness_Mobile}
   Alpha=NamedRerouteUsage()
  )
  NamedRerouteDeclaration(None=LinearInterpolate(
    A=ShadingPathSwitch(Default=@n31, Mobile=@n32)
    B=ScalarParameter() {defaultValue=0.9, group=Morphite, parameterName=RoughnessSpecRemoval}
    Alpha=NamedRerouteUsage()
   )) {name=Roughness, nodeColor=<rgba 1 0.3984 0 1>}
  n33 = DotProduct(A=NamedRerouteUsage(), B=Constant3Vector() {constant=<rgba 0 0 1 0>})
  NamedRerouteDeclaration(None=SphereMask(A=Saturate(None=@n33), B=Constant() {r=1}, Radius=Constant() {r=0.2}, Hardness=Constant() {r=0.5})) {name=VertexNormalMask.Z, nodeColor=<rgba 0 0.3203 1 1>}
  NamedRerouteDeclaration(None=ScalarParameter() {defaultValue=0.65, group=Morphite, parameterName=Metallic}) {name=Metallic, nodeColor=<rgba 0.5 1 0 1>}
  NamedRerouteDeclaration(None=Transform(None=@n3)) {name=ChromeNormalsWS, nodeColor=<rgba 0.8516 0 1 1>}
  NamedRerouteDeclaration(None=FunctionInput(Preview=VertexNormalWS()) {bUsePreviewValueAsDefault=true, inputName=VertexNormalOverride, sortPriority=4}) {name=VertexNormal, nodeColor=<rgba 0.5 0 1 1>}
 ]}
