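# Material function for Morphite weapons that progressively blends the incoming material into MF_Morphite_Base chrome along the weapon's length.
# Custom primitive data supplies phase, transition and flash times, and level-up availability, driving a chrome mask along pre-skinned Y.
# Grid-texture noise breaks up emissive transition glows, flash sweeps and a level-up pulse.
# Panning swirl normals and derivative-based edge normals add detail, with cheaper mobile and low-quality paths.
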MaterialFunction /BlastBerryWeapons/Materials/Morphite/MF_Morphite_Weapon_Transition
graph {nodes [
  n1 = GetMaterialAttributes(None=FunctionInput() {inputType=FunctionInput_MaterialAttributes}) {attributeGetTypes [#1, #2]}
  n2 = GetMaterialAttributes(None=MaterialFunctionCall!(Override Normals=NamedRerouteUsage(), UVsOverride, BaseColorOverride, WorldPositionOverride=NamedRerouteUsage(), VertexNormalOverride) {materialFunction=&/MF_Morphite_Base}.MA Out) {attributeGetTypes [#3]}
  n3 = NamedRerouteUsage()
  n4 = NamedRerouteUsage()
  n5 = Add(A=Multiply(A=NamedRerouteUsage()) {constB=4}, B=Multiply(A=Multiply(A=@n3), B=Multiply(A=@n4) {constB=0.2}))
  n6 = MaterialFunctionCall!(
   Material=SetMaterialAttributes!(MaterialAttributes=@n2.MaterialAttributes, EmissiveColor=Multiply(A=@n2.EmissiveColor, B=@n5) {constB=0}) {attributeSetTypes [#3]}
   NewBaseColor=NamedRerouteUsage()
   Mask
  ) {materialFunction=/Engine/Functions/MaterialLayerFunctions/MatLayerBlend_MultiplyBaseColor}
  n7 = NamedRerouteUsage()
  n8 = FunctionInput() {bUsePreviewValueAsDefault=true, inputName=Mask, inputType=FunctionInput_Scalar, sortPriority=1000, previewValue=<1 0 0 1>}
  n9 = GetMaterialAttributes(None=BlendMaterialAttributes(A=@n1.MaterialAttributes, B=@n6.Blended Material, Alpha=Multiply(A=@n7, B=@n8))) {attributeGetTypes [#3]}
  n10 = PreSkinnedPosition()
  n11 = VertexInterpolator(VS=@n10)
  n12 = Subtract(
   A=ComponentMask(None=ShaderStageSwitch(PixelShader=@n11.PS, VertexShader=@n10)) {g=true}
   B=ScalarParameter() {group=Morphite, parameterName=WeaponLengthOffset}
  )
  n13 = ScalarParameter() {defaultValue=15, group=Morphite, parameterName=WeaponLength}
  n14 = Power(Base=NamedRerouteUsage())
  n15 = Add(
   A=ScalarParameter() {defaultValue=15, group=Morphite, parameterName=WeaponLength}
   B=ScalarParameter() {defaultValue=50, group=Morphite, parameterName=ChromeFlashDistance}
  )
  n16 = Divide(
   A=Add(A=Subtract(A=@n12, B=@n13), B=Multiply(A=@n14, B=@n15))
   B=ScalarParameter() {defaultValue=50, group=Morphite, parameterName=ChromeFlashDistance}
  )
  n17 = OneMinus(None=Saturate(None=@n16))
  n18 = Sign(None=@n16)
  n19 = Power(Base=Saturate(None=Multiply(A=@n17, B=@n18)))
  n20 = Time()
  n21 = ScalarParameter() {defaultValue=15, group=Morphite, parameterName=WeaponLength}
  n22 = Divide(A=@n12, B=@n21)
  n23 = Multiply(A=@n22) {constB=0.7}
  n24 = ConstantBiasScale(None=Sine(None=Add(A=@n20, B=@n23)) {period=1.5})
  n25 = Multiply(
   A=Multiply(A=Power(Base=@n24) {constExponent=30}) {constB=0.75}
   B=ScalarParameter() {bUseCustomPrimitiveData=true, group=Morphite, parameterName=WeaponLevelUpAvailable, primitiveDataIndex=4}
  )
  n26 = TextureObject() {samplerType=SAMPLERTYPE_LinearColor, texture=/Game/Packages/Fortress_SharedMaterials/Textures/T_Slowfield_GridSquares_01}
  n27 = Constant3Vector() {constant=<rgba 50 50 50 0>}
  n28 = ScalarParameter() {defaultValue=15, group=Morphite, parameterName=WeaponLength}
  n29 = ScalarParameter() {group=Morphite, parameterName=UVOffset}
  n30 = Add(
   A=NamedRerouteUsage()
   B=MaterialFunctionCall(B=Add(A=@n28, B=@n29)) {materialFunction=&AppendMany}.RGB
  )
  n31 = Frac(None=ScalarParameter() {bUseCustomPrimitiveData=true, group=Morphite, parameterName=ChromeFlashTime, primitiveDataIndex=3})
  n32 = Divide(A=Floor(None=Multiply(A=@n31) {constB=16})) {constB=16}
  n33 = NamedRerouteUsage()
  n34 = MaterialFunctionCall(TextureObject=@n26, TextureSize=@n27, WorldPosition=Add(A=@n30, B=Multiply(A=@n32, B=@n27)), World Space Normal=@n33) {materialFunction=/Game/Weapons/WeaponSkins/Wraps/MaterialFunctions/WorldAlignedTexture}
  n35 = ComponentMask(None=MaterialFunctionCall(TextureObject=@n26, TextureSize=@n27, WorldPosition=@n30, World Space Normal=@n33) {materialFunction=&WorldAlignedTexture}.XY Texture) {r=true}
  n36 = QualitySwitch(Default=ShadingPathSwitch(Default=ComponentMask(None=@n34.XY Texture) {r=true}, Mobile=@n35), Low=@n35, Medium=@n35)
  n37 = Multiply(A=Subtract(A=OneMinus(None=@n36)) {constB=0.33}) {constB=1.5}
  n38 = NamedRerouteUsage()
  n39 = Multiply(A=Multiply(A=Max(A=@n19, B=@n25), B=@n37), B=OneMinus(None=Multiply(A=@n38) {constB=0.5}))
  n40 = SetMaterialAttributes(MaterialAttributes=@n9.MaterialAttributes, EmissiveColor=Add(A=@n9.EmissiveColor, B=Multiply(A=@n2.EmissiveColor, B=@n39))) {attributeSetTypes [#3]}
  n41 = ShadingPathSwitch(
   Default=MaterialFunctionCall!(Material=@n40, Additional Normal=NamedRerouteUsage(), Mask) {materialFunction=/Engine/Functions/MaterialLayerFunctions/MatLayerBlend_BlendAngleCorrectedNormals}.Blended Material
   Mobile=@n40
  )
  FunctionOutput(None=MaterialFunctionCall!(Material=QualitySwitch(Default=@n41, Low=@n40), Mask, WorldPositionOffset=NamedRerouteUsage()) {materialFunction=/Engine/Functions/MaterialLayerFunctions/MatLayerBlend_OverrideWorldPositionOffset}.Blended Material)
  MaterialFunctionCall(Material=@n6.Blended Material, Additional Normal=NamedRerouteUsage()) {materialFunction=&MatLayerBlend_BlendAngleCorrectedNormals}
  n42 = MaterialFunctionCall() {materialFunction=/Game/Packages/Fortress_SharedMaterials/Base_Material_Functions/BaseCustomizationFunctions/MF_Base_LocalPosition}
  n43 = LocalPosition() {localOrigin=InstancePreSkinning}
  NamedRerouteDeclaration(None=FunctionInput(Preview=ShaderStageSwitch(PixelShader=@n42.Pre-Skinned Local Position Pixel Shader, VertexShader=@n43.XYZ)) {bUsePreviewValueAsDefault=true, inputName=Preskinned Position, sortPriority=50}) {name=Preskinned Position, nodeColor=<rgba 0.121 0.2774 1 1>}
  n44 = PreSkinnedNormal()
  NamedRerouteDeclaration(None=FunctionInput(Preview=ShaderStageSwitch(PixelShader=@n42.Pre-Skinned Local Normal Pixel Shader, VertexShader=@n44)) {bUsePreviewValueAsDefault=true, inputName=Preskinned Normals, sortPriority=60}) {name=Preskinned Normals, nodeColor=<rgba 0.121 0.2774 1 1>}
  n45 = ScalarParameter() {defaultValue=4, group=Morphite, parameterName=WeaponPhaseAmount}
  n46 = OneMinus(None=NamedRerouteUsage())
  n47 = Max(
   A=ScalarParameter() {group=Morphite, parameterName=WeaponPhasePreview, primitiveDataIndex=2}
   B=Subtract(A=ScalarParameter() {bUseCustomPrimitiveData=true, group=Morphite, parameterName=WeaponPhase, primitiveDataIndex=1}, B=Power(Base=@n46))
  )
  n48 = Subtract(A=@n45, B=@n47)
  n49 = Subtract(A=@n12, B=Multiply(A=@n21, B=Divide(A=@n48, B=@n45)))
  n50 = Subtract(
   A=Divide(A=@n49, B=ScalarParameter() {defaultValue=30, group=Morphite, parameterName=ChromeTransitionDistance}) {constB=20}
   B=NamedRerouteUsage()
  )
  n51 = ScalarParameter() {defaultValue=10, group=Morphite, parameterName=ChromeTransitionHardness}
  n52 = NamedRerouteDeclaration(None=Saturate(None=Multiply(A=@n50, B=@n51))) {name=Chrome Mask, nodeColor=<rgba 0.0078 1 0 1>}
  n53 = Add(A=@n52, B=Multiply(A=@n52, B=Multiply(A=@n37) {constB=0.05}))
  n54 = MaterialFunctionCall(X=DDX(Value=@n53), Y=DDY(Value=@n53), Z=ScalarParameter() {defaultValue=0.1, group=Morphite, parameterName=Edge Z Normal}) {materialFunction=&MakeFloat3}
  n55 = Constant3Vector() {constant=<rgba -1 -1 1 0>}
  NamedRerouteDeclaration(None=Normalize(VectorInput=Multiply(A=@n54.Result, B=@n55))) {name=Edge Normal, nodeColor=<rgba 0.7344 1 0 1>}
  NamedRerouteDeclaration(None=TextureCoordinate()) {name=UVs, nodeColor=<rgba 0.2891 0 1 1>}
  NamedRerouteDeclaration(None=@n1.WorldPositionOffset) {name=Original WPO, nodeColor=<rgba 0.2891 1 0 1>}
  n56 = TextureSampleParameter2D(UVs=NamedRerouteUsage()) {
   group=Morphite
   parameterName=ChromeNormals
   samplerSource=SSM_Wrap_WorldGroupSettings
   samplerType=SAMPLERTYPE_Normal
   texture=/Game/TechArt/Seasonal/LackLuster/Textures/T_Lackluster_Swirl_N
  }
  n57 = TextureSampleParameter2D(UVs=NamedRerouteUsage()) {
   group=Morphite
   parameterName=ChromeNormals
   samplerSource=SSM_Wrap_WorldGroupSettings
   samplerType=SAMPLERTYPE_Normal
   texture=&T_Lackluster_Swirl_N
  }
  n58 = ShadingPathSwitch(Default=Add(A=@n56.RGB, B=@n57.RGB), Mobile=@n56.RGB)
  n59 = ScalarParameter() {defaultValue=0.8, group=Morphite, parameterName=Flatten Swirl Facing}
  n60 = ScalarParameter() {defaultValue=0.7, group=Morphite, parameterName=Flatten Swirl Side}
  n61 = Fresnel(Normal=VertexNormalWS()) {baseReflectFraction=0, exponent=3}
  NamedRerouteDeclaration(None=MaterialFunctionCall(Normal=Normalize(VectorInput=@n58), Flatness=LinearInterpolate(A=@n59, B=@n60, Alpha=@n61)) {materialFunction=&FlattenNormal}.Result) {name=Swirl Normal, nodeColor=<rgba 1 0.0469 0 1>}
  n62 = ScalarParameter() {defaultValue=0.5, group=Morphite, parameterName=ChromeTransitionDuration}
  n63 = Divide(
   A=Subtract(
    A=Time() {period=2}
    B=ScalarParameter() {bUseCustomPrimitiveData=true, group=Morphite, parameterName=ChromeTransitionTime, primitiveDataIndex=2}
   )
   B=@n62
  )
  NamedRerouteDeclaration(None=Saturate(None=@n63)) {name=GrowthAnimation, nodeColor=<rgba 0 1 0.7422 1>}
  NamedRerouteDeclaration(None=Saturate(None=Subtract(A=@n63, B=@n62))) {name=EmissiveFalloffAnimation, nodeColor=<rgba 0 1 0.7422 1>}
  n64 = Subtract(A=Time() {period=2}, B=ScalarParameter() {bUseCustomPrimitiveData=true, group=Morphite, parameterName=ChromeFlashTime, primitiveDataIndex=3})
  n65 = ScalarParameter() {defaultValue=0.75, group=Morphite, parameterName=ChromeFlashDuration}
  NamedRerouteDeclaration(None=Saturate(None=Divide(A=@n64, B=@n65))) {name=FlashAnimation, nodeColor=<rgba 0 1 0.7422 1>}
  n66 = ScalarParameter() {defaultValue=30, group=Morphite, parameterName=GradientSize}
  n67 = Divide(A=@n49, B=Divide(A=@n66) {constB=0.5})
  NamedRerouteDeclaration(None=OneMinus(None=Saturate(None=@n67))) {name=Gradient, nodeColor=<rgba 0 1 0.8359 1>}
  n68 = PreSkinnedLocalBounds()
  NamedRerouteDeclaration(None=FunctionInput(Preview=ShaderStageSwitch(PixelShader=@n42.Pre-Skinned Local Bounds Pixel Shader, VertexShader=@n68.Half Extents)) {bUsePreviewValueAsDefault=true, inputName=Preskinned Bounds, sortPriority=70}) {name=Preskinned Bounds, nodeColor=<rgba 0.121 0.2774 1 1>}
  NamedRerouteDeclaration(None=Divide(A=@n47, B=@n45)) {name=PhaseNormalized, nodeColor=<rgba 0.8047 0 1 1>}
  n69 = Time()
  n70 = Frac(None=Add(A=NamedRerouteUsage(), B=Divide(A=@n69) {constB=4}))
  n71 = NamedRerouteUsage()
  n72 = Multiply(A=Power(Base=OneMinus(None=@n70), Exp=LinearInterpolate(Alpha=@n71) {constA=5, constB=2}) {constExponent=5})
  n73 = Power(Base=NamedRerouteUsage())
  NamedRerouteDeclaration(None=Saturate(None=Multiply(A=@n72, B=@n73))) {name=Noise Emissive, nodeColor=<rgba 0 0.1391 1 1>}
  n74 = NamedRerouteDeclaration(None=@n35) {name=Noise, nodeColor=<rgba 1 0 0.8203 1>}
  NamedRerouteDeclaration(None=Multiply(A=ComponentMask(None=@n74) {r=true}, B=ScalarParameter() {defaultValue=0.6, group=Morphite, parameterName=Growth Noise}) {constB=0.1}) {name=Transition Mask, nodeColor=<rgba 1 0.375 0 1>}
  n75 = NamedRerouteUsage()
  n76 = Multiply(A=ScalarParameter() {defaultValue=1, group=Morphite, parameterName=ChromePanSpeedScale}, B=Time())
  NamedRerouteDeclaration(None=Panner(Coordinate=NamedRerouteDeclaration(None=@n75) {name=Chrome_UVs_A_NoPanning, nodeColor=<rgba 0 1 0.7656 1>}, Time=@n76) {bFractionalPart=true, speedX=0.003, speedY=0.006}) {name=Chrome_UVs_A, nodeColor=<rgba 0.9219 0 1 1>}
  n77 = Saturate(None=Divide(A=@n49, B=@n66) {constB=30})
  n78 = Divide(B=ScalarParameter() {defaultValue=4, group=Morphite, parameterName=GradientPower})
  NamedRerouteDeclaration(None=LinearInterpolate(
    A=ScalarParameter() {defaultValue=5, group=Morphite, parameterName=GradientMax}
    B=ScalarParameter() {defaultValue=1, group=Morphite, parameterName=GradientMin}
    Alpha=Power(Base=@n77, Exp=@n78)
   )) {name=GradientTransform, nodeColor=<rgba 1 0.2578 0 1>}
  NamedRerouteDeclaration(None=Panner(Coordinate=NamedRerouteUsage(), Time=@n76) {bFractionalPart=true, speedX=0.015, speedY=0.006}) {name=Chrome_UVs_B, nodeColor=<rgba 0 0.3906 1 1>}
  NamedRerouteDeclaration(None=@n1.Normal) {name=Normal In, nodeColor=<rgba 0 0.4141 1 1>}
  n79 = Min(A=NamedRerouteUsage(), B=OneMinus(None=NamedRerouteUsage()))
  n80 = OneMinus(None=Saturate(None=@n50))
  NamedRerouteDeclaration(None=Multiply(A=Power(Base=@n79), B=Power(Base=@n80) {constExponent=4})) {name=EmissiveTransition, nodeColor=<rgba 0.3359 1 0 1>}
  SetMaterialAttributes!(MaterialAttributes, BaseColor=@n22) {attributeSetTypes [#4]}
 ]}
