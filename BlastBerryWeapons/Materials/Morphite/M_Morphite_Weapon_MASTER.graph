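# Master material for Morphite weapons, adding a Morphite transition effect to the shared weapon master shading.
# MF_FN_Weapon_MASTER feeds OverheatGunFunction, and a shading path switch bypasses the overheat layer on mobile.
# The switched attributes pass through MF_Morphite_Weapon_Transition, then ADSScopeMaster.
# It is Default Lit, uses material attributes, flags pixel animation, and supports Nanite, skeletal mesh, cloth, morph target and mesh deformer usage.
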Material /BlastBerryWeapons/Materials/Morphite/M_Morphite_Weapon_MASTER
graph {
 properties {
  bAutomaticallySetUsageInEditor=false
  bHasPixelAnimation=true
  bUseMaterialAttributes=true
  bUsedWithClothing=true
  bUsedWithMeshDeformer=true
  bUsedWithMorphTargets=true
  bUsedWithNanite=true
  bUsedWithSkeletalMesh=true
  pixelDepthOffsetMode=PDOM_Legacy
  usedShadingModels=Default Lit
 }
 nodes [
  n1 = MaterialFunctionCall!(In) {materialFunction=/Game/Packages/Fortress_SharedMaterials/Base_Material_Functions/BaseCustomizationFunctions/MF_FN_Weapon_MASTER}
  n2 = MaterialFunctionCall!(Base MA=@n1.Material Result, Pre-Skinned Local Position=@n1.Material Result) {materialFunction=/Game/Weapons/WeaponSkins/MaterialFunction/OverheatGunFunction}
  MaterialFunctionCall!(Material Attributes=MaterialFunctionCall!(
    In=ShadingPathSwitch(Default=@n2.Result, Mobile=@n1.Material Result) {desc=Migrated from FeatureLevelSwitch}
    Preskinned Position
    Preskinned Normals
    Preskinned Bounds
    Mask
   ) {materialFunction=&/MF_Morphite_Weapon_Transition}.Result) {materialFunction=/Game/Packages/Fortnite_UniqueMaterials_1/Weapons/ADSScopeMaster}
 ]
}
